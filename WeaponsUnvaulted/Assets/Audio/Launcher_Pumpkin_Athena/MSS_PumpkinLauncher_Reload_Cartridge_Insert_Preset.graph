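# MetaSound Source preset for the Pumpkin Launcher's cartridge-insert reload sound, a mono one-shot.
# It wraps the shared MSS_Weapon_Reload_Insert graph, passing in three insert wave variations, a master volume and a random pitch-shift base and spread.
# The parent's audio and finished trigger pass straight out, played through Fortnite weapon-foley sound class, attenuation and concurrency.
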MetaSoundSource /WeaponsUnvaulted/Assets/Audio/Launcher_Pumpkin_Athena/MSS_PumpkinLauncher_Reload_Cartridge_Insert_Preset
inputs [
 MasterVolume {type=Float}
 Random.PitchShift.Base {type=Float}
 Random.PitchShift.Spread {type=Float}
 Sounds {type=WaveAsset:Array}
 UE.Source.OnPlay {type=Trigger}
]
interfaces [UE.OutputFormat.Mono:1.0, UE.Source.OneShot:1.0, UE.Source:1.1]
outputs [
 "UE.OutputFormat.Mono.Audio:0" {type=Audio}
 UE.Source.OneShot.OnFinished {type=Trigger}
]
pages [Default {
  input_defaults {
   MasterVolume=0.7
   Random.PitchShift.Base=0
   Random.PitchShift.Spread=0.5
   UE.Source.OnPlay=null
   Sounds [
    /WeaponsUnvaulted/Assets/Audio/Launcher_Pumpkin_Athena/SoundWaves/PumpkinLauncher_Reload_Cartridge_Insert_03
    /WeaponsUnvaulted/Assets/Audio/Launcher_Pumpkin_Athena/SoundWaves/PumpkinLauncher_Reload_Cartridge_Insert_02
    /WeaponsUnvaulted/Assets/Audio/Launcher_Pumpkin_Athena/SoundWaves/PumpkinLauncher_Reload_Cartridge_Insert_01
   ]
  }
  nodes [
   InitVariable.WaveAsset:Array()
   MSS_Weapon_Reload_Insert_0 = /FortMetasound/Weapons/MSS_Weapon_Reload_Insert(
    MasterVolume=^MasterVolume
    Random.PitchShift.Base=^Random.PitchShift.Base
    Random.PitchShift.Spread=^Random.PitchShift.Spread
    Sounds=^Sounds
    UE.Source.OnPlay=^UE.Source.OnPlay
   )
  ]
  links [
   {dst="@UE.OutputFormat.Mono.Audio:0", src=MSS_Weapon_Reload_Insert_0.UE.OutputFormat.Mono.Audio:0}
   {dst="@UE.Source.OneShot.OnFinished", src=MSS_Weapon_Reload_Insert_0.UE.Source.OneShot.OnFinished}
  ]
 }]
properties {
 assetImportData=&self:AssetImportData
 attenuationSettings=&AS_Fort_WeaponFoley
 maxDistance=2250
 sampleRate=48000
 soundClassObject=&Fort_WeaponFoley
 volume=1.1
 concurrencySet [&CS_Fort_WeaponFoley]
}
